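# Post-process style part AnimBlueprint for the Buff Cat Comic body, adding ear jiggle and emote-driven mouth changes.
# AnimGraph copies pose and curves from the attached parent mesh, then chains Anim Dynamics on the left and right ear bones, tuned separately.
# Each update sets the dynamic material's smile-mouth frame according to whether an emote is playing.
# A facial-sequence blend weighted by the DisableFaceOverride curve is present but not connected to the output.
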AnimBlueprint /BRCosmetics/Characters/Player/Male/Medium/Bodies/M_MED_Buff_Cat_Comic/Meshes/Parts/M_MED_Buff_Cat_Comic_AnimBP
target_skeleton=&/M_MED_Buff_Cat_Comic_Skeleton
graphs [
 AnimGraph {nodes [
   Layered blend per bone(
    BlendPoses_0=Sequence Player 'BuffCat_Male_Facial'() {sequence=/BRCosmetics/Animation/Game/MainPlayer/Facial/BuffCat/BuffCat_Male_Facial}
    BlendWeights_0=float * float(A=1, B=GetCurveValue(CurveName=DisableFaceOverride))
   )
   n1 = Anim Dynamics - Bone: dyn_ear_l(
    ComponentPose=Local To Component(LocalPose=Copy Pose From Mesh() {bCopyCurves=true, bUseAttachedParent=true})
    Alpha=1
    ExternalForce="0,0,0"
   ) {
    bGravityOverrideInSimSpace=true
    bUseGravityOverride=true
    simulationSpace=Actor
    alphaBoolBlend {blendOption=Linear}
    boundBone {boneName=dyn_ear_l}
    componentAppliedLinearAccClamp=<2000 2000 2000>
    componentLinearAccScale=<0.5 0.5 0.5>
    componentLinearVelScale=<0.5 0.5 0.5>
    gravityOverride=<1 1 0.85>
    physicsBodyDefinitions [{
      sphereCollisionRadius=0
      boundBone {boneName=dyn_ear_l}
      boxExtents=<6 6 2>
      constraintSetup {angularLimitsMax=<0 25 0>, angularLimitsMin=<0 1 0>, angularTarget=<0 0 0>}
      localJointOffset=<2 0 0>
     }]
    relativeSpaceBone {boneName=spine_05}
   }
   n2 = Anim Dynamics - Bone: dyn_ear_r(ComponentPose=@n1, Alpha=0.75, ExternalForce="0,0,0") {
    bGravityOverrideInSimSpace=true
    bUseGravityOverride=true
    alphaBoolBlend {blendOption=Linear}
    boundBone {boneName=dyn_ear_r}
    componentAppliedLinearAccClamp=<2000 2000 2000>
    componentLinearAccScale=<0.25 0.25 0.25>
    componentLinearVelScale=<0.25 0.25 0.25>
    gravityOverride=<1 1 0.75>
    physicsBodyDefinitions [{
      sphereCollisionRadius=0
      boundBone {boneName=dyn_ear_r}
      boxExtents=<6 6 2>
      constraintSetup {angularLimitsMax=<0 3 0>, angularLimitsMin=<0 -15 0>, angularTarget=<0 0 0>}
      localJointOffset=<2 0 0>
     }]
    relativeSpaceBone {boneName=pelvis}
   }
   Output Pose(Result=Component To Local(ComponentPose=@n2)) {layerGroup=None}
  ]}
 EventGraph {nodes [
   Event BlueprintUpdateAnimation()
   > n1 = Get Dynamic Material Instance(SkeletalMeshComponent=GetPartSkeletalMeshComponent())
   > n2 = Branch(Condition=Get bIsPlayingEmote())
   > SetScalarParameterValue(self=@n1, ParameterName=Smile Mouth animation Frame, Value=8)
   SetScalarParameterValue(execute=@n2.else, self=@n1, ParameterName=Smile Mouth animation Frame, Value=1)
  ]}
]
variables []
